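# Translucent unlit mesh-particle material for Geode "BitTrail" electrical arcs, tinted by Geode cloud and crack colors through a mask that fades at both UV0 V ends.
# A sine pulse or manual value selects which quantized UV1 band shows, gating color, opacity and offset.
# World position offset combines SplineThicken, widened with camera distance, and time-stepped normal-noise wobble.
# ES2/ES3_1 drop the wobble and depth-fade color boost.
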Material /Figment_S10/Athena/Effects/Materials/Geode/M_Geode_2_BitTrail
graph {
 properties {
  bUsedWithMeshParticles=true
  blendMode=BLEND_Translucent
  pixelDepthOffsetMode=PDOM_Legacy
  shadingModel=MSM_Unlit
  twoSided=true
  usedShadingModels=Unlit
 }
 nodes [
  n1 = MaterialFunctionCall!(Use Alternate UVs, Alternate UVs) {materialFunction=/Game/Athena/Effects/Materials/Geode/Functions/MF_Geode_Crack_Colors}
  n2 = ComponentMask(None=TextureCoordinate()) {g=true}
  n3 = OneMinus(None=@n2)
  n4 = Multiply(
   A=Power(Base=Multiply(A=@n3, B=@n2), Exp=ScalarParameter() {defaultValue=1, group=Color, parameterName=Mask Exponent}) {constExponent=3}
   B=ScalarParameter() {defaultValue=50, group=Color, parameterName=Mask Scale}
  ) {constB=50}
  n5 = Clamp(None=@n4)
  n6 = LinearInterpolate(A=Saturate(None=@n1.Cloud Color), B=Saturate(None=@n1.Crack Color), Alpha=@n5)
  n7 = LinearInterpolate(
   A=@n6
   B=Multiply(B=@n6) {constA=2}
   Alpha=DepthFade(Opacity=@n5, FadeDistance=ScalarParameter() {defaultValue=2, parameterName=ContactGlowDistance})
  )
  n8 = Multiply(A=Time(), B=ScalarParameter() {defaultValue=3, parameterName=Pulse Speed})
  n9 = LinearInterpolate(
   A=ConstantBiasScale(None=Sine(None=@n8))
   B=ScalarParameter() {defaultValue=0.26, parameterName=Manual Pulse Selection}
   Alpha=ScalarParameter() {parameterName=Use Manual Selection}
  )
  n10 = ComponentMask(None=TextureCoordinate() {coordinateIndex=1}) {g=true}
  n11 = ScalarParameter() {defaultValue=4, parameterName=NumberofPhases_1}
  n12 = Divide(A=Ceil(None=Multiply(A=@n10, B=@n11)), B=@n11)
  n13 = Floor(None=Add(A=@n12) {constB=0.5})
  n14 = Subtract(A=Floor(None=Add(A=@n12) {constB=0.75}), B=@n13)
  n15 = Constant() {r=0.5}
  n16 = Floor(None=Add(A=@n12) {constB=0.25})
  n17 = Subtract(A=@n13, B=@n16)
  n18 = Floor(None=@n12)
  n19 = If(A=@n9, B=Constant() {r=0.25}, "A > B"=Subtract(A=@n16, B=@n18), "A == B"=@n18, "A < B"=@n18)
  n20 = Multiply(
   A=ScalarParameter() {defaultValue=1, parameterName=Electrical Arcs}
   B=If(A=@n9, B=Constant() {r=0.75}, "A > B"=@n14, "A == B"=@n14, "A < B"=If(A=@n9, B=@n15, "A > B"=@n17, "A == B"=@n17, "A < B"=@n19))
  )
  n21 = Multiply(A=FeatureLevelSwitch(Default=@n7, ES2=@n6, ES3_1=@n6, SM4_REMOVED=@n7), B=@n20)
  MaterialFunctionCall(LDR Input=Clamp(None=@n21), HDR Tint=@n21, HDR Intensity Multiplier=ScalarParameter() {defaultValue=0.5, parameterName=HDR Intensity}) {materialFunction=&DeriveHDRfromLDR}
  Multiply(
   A=Multiply(A=@n5, B=@n20)
   B=MaterialFunctionCall!(In) {materialFunction=/Game/Effects/Fort_Effects/Materials/Geode/_Functions/MF_Geode_Fade}.Opacity_HeatShield Loca
  )
  n22 = ObjectPositionWS()
  n23 = CameraPositionWS()
  n24 = Clamp(
   None=Divide(A=Distance(A=@n22, B=@n23), B=ScalarParameter() {defaultValue=1400, group=Scale By Distance, parameterName=Distance Scale})
   Min=ScalarParameter() {group=Scale By Distance, parameterName=DistanceScaleMinClamp}
   Max=ScalarParameter() {defaultValue=50, group=Scale By Distance, parameterName=DistanceScaleMaxClamp}
  ) {maxDefault=10}
  n25 = TextureCoordinate()
  n26 = TextureCoordinate() {vTiling=2}
  n27 = Add(
   A=TextureSample(UVs=Panner(Coordinate=@n25) {speedY=1}) {texture=/BRCosmetics/Effects/Fort_Effects/Effects/Characters/Generic/ResIn/CylGradient}.G
   B=TextureSample(UVs=Panner(Coordinate=@n26) {speedY=1}) {texture=&CylGradient}.G
  )
  n28 = TextureSample(UVs=Panner(Coordinate=TextureCoordinate() {vTiling=6}) {speedY=1}) {texture=&CylGradient}
  n29 = MaterialFunctionCall(
   WidthBase=Add(A=@n24, B=ScalarParameter() {defaultValue=2, group=WPO, parameterName=WidthBase})
   WidthTip=Add(A=ScalarParameter() {defaultValue=2, group=WPO, parameterName=WidthTop}, B=@n24)
   UVs for Thickness=Multiply(A=Add(A=@n27, B=@n28.G), B=ScalarParameter() {defaultValue=2.5, group=WPO, parameterName=LengthNoise})
  ) {materialFunction=&SplineThicken}
  n30 = Time()
  n31 = ScalarParameter() {defaultValue=8, group=WPO, parameterName=StepSpeed}
  n32 = Divide(A=Round(None=Multiply(A=@n30, B=@n31)), B=@n31)
  n33 = Panner(Coordinate=TextureCoordinate() {uTiling=0.25, vTiling=0.25}, Time=@n32) {speedX=0.25, speedY=0.25}
  n34 = TextureCoordinate()
  n35 = Add(
   A=Multiply(
    A=TextureSample(UVs=@n33) {samplerType=SAMPLERTYPE_Normal, texture=/Figment_S10/Packages/User_Built_Props/Textures/T_Normal_Noise01}.RGB
    B=Constant3Vector() {constant=<rgba 0.2 0.2 0.5 0>}
   )
   B=TextureSample(UVs=Panner(Coordinate=@n34, Time=@n32) {speedX=0.2, speedY=0.2}) {samplerType=SAMPLERTYPE_Normal, texture=&T_Normal_Noise01}.RGB
  )
  n36 = Subtract(A=Multiply(A=ComponentMask(None=@n35) {r=true}) {constB=2})
  n37 = Subtract(A=Multiply(A=ComponentMask(None=@n35) {g=true}) {constB=2})
  n38 = Multiply(A=ComponentMask(None=@n35) {b=true}) {constB=2}
  n39 = Multiply(
   A=AppendVector(A=AppendVector(A=@n36, B=@n37), B=Subtract(A=@n38))
   B=Transform(None=VectorParameter() {group=WPO, parameterName=WobbleOffset, defaultValue=<rgba 25 25 50 1>}.RGB) {transformSourceType=TRANSFORMSOURCE_World, transformType=TRANSFORM_Local}
  )
  n40 = ScalarParameter() {defaultValue=0.15, group=WPO, parameterName=NoiseScale}
  n41 = Clamp(None=Multiply(A=@n4) {constB=10}) {minDefault=0.05}
  n42 = Multiply(
   A=Multiply(A=Multiply(A=@n39, B=@n40), B=@n41)
   B=MaterialFunctionCall() {materialFunction=/Engine/Functions/Engine_MaterialFunctions02/WorldPositionOffset/ObjectScale}.Scale XYZ
  )
  n43 = Transform(None=VectorParameter() {group=WPO, parameterName=PivotOffset}.RGB) {transformSourceType=TRANSFORMSOURCE_Local}
  n44 = Add(A=Multiply(A=Add(A=@n29.WorldPositionOffset, B=@n42), B=@n41), B=@n43)
  n45 = Add(A=@n29.WorldPositionOffset, B=@n43)
  Multiply(A=@n20, B=FeatureLevelSwitch(Default=@n44, ES2=@n45, ES3_1=@n45, SM4_REMOVED=@n44))
 ]
}
