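# Translucent unlit sprite and beam-trail particle material for Geode effects, named as a darkness layer.
# A powered radial sphere mask, particle alpha and Geode fade shape opacity, with depth and near-camera fades replaced by a constant on ES2/ES3_1.
# A UseHDR switch picks particle color or HDR-remapped crack Cloud Color times particle color.
# The contrast-driven color lerp has identical endpoints, so it has no visible effect.
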Material /Figment_S10/Athena/Effects/Materials/Geode/M_Geode_2_Darkness
graph {
 properties {
  bUsedWithBeamTrails=true
  bUsedWithParticleSprites=true
  blendMode=BLEND_Translucent
  pixelDepthOffsetMode=PDOM_Legacy
  shadingModel=MSM_Unlit
  usedShadingModels=Unlit
 }
 nodes [
  n1 = Constant() {r=0.5}
  n2 = SphereMask(A=TextureCoordinate(), B=@n1, Radius=@n1) {hardnessPercent=0}
  n3 = ParticleColor()
  n4 = DepthFade(
   Opacity=MaterialFunctionCall() {materialFunction=&Near_Camera_Fade}.Result
   FadeDistance=ScalarParameter() {defaultValue=200, parameterName=DepthFade}
  )
  n5 = Constant() {r=1}
  n6 = FeatureLevelSwitch(Default=@n4, ES2=@n5, ES3_1=@n5, SM4_REMOVED=@n4, SM5=@n4, SM6=@n4)
  Multiply(
   A=Multiply(A=Power(Base=@n2) {constExponent=3}, B=Multiply(A=@n3.A, B=@n6))
   B=MaterialFunctionCall!(In) {materialFunction=/Game/Effects/Fort_Effects/Materials/Geode/_Functions/MF_Geode_Fade}.Opacity_XAxis Particle Sprites
  )
  n7 = MaterialFunctionCall!(Use Alternate UVs, Alternate UVs) {materialFunction=/Game/Athena/Effects/Materials/Geode/Functions/MF_Geode_Crack_Colors}
  n8 = ScalarParameter() {defaultValue=4, parameterName=Color Contrast}
  n9 = Multiply(A=LinearInterpolate(A=@n7.Cloud Color, B=@n7.Cloud Color, Alpha=Power(Base=@n2, Exp=@n8) {constExponent=4}), B=@n3.RGB)
  StaticSwitchParameter(
   True=MaterialFunctionCall(LDR Input=Clamp(None=@n9), HDR Tint=@n9) {materialFunction=&DeriveHDRfromLDR}.RemappedRange
   False=@n3.RGB
  ) {parameterName=UseHDR}
 ]
}
