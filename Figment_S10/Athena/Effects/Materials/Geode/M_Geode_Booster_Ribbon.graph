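# Additive, unlit beam-trail material for the Geode booster ribbon, with color and alpha driven by particle color.
# Turbulence sampled at offset UVs forms RGB-split noise, whose desaturated value drives erosion through the dynamic Erode parameter.
# Crack and cloud colors blend along U with panning dust, which also boosts HDR intensity.
# Edge fades and normal-map Fresnel shape a subtle refraction-style value; mobile switches bypass it and the HDR remap.
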Material /Figment_S10/Athena/Effects/Materials/Geode/M_Geode_Booster_Ribbon
graph {
 properties {
  bEnableResponsiveAA=true
  bUsedWithBeamTrails=true
  blendMode=BLEND_Additive
  pixelDepthOffsetMode=PDOM_Legacy
  refractionMethod=RM_IndexOfRefraction
  shadingModel=MSM_Unlit
  usedShadingModels=Unlit
 }
 nodes [
  n1 = ParticleColor()
  n2 = TextureCoordinate() {coordinateIndex=1, uTiling=0.25}
  n3 = AppendVector(A=ScalarParameter() {defaultValue=0.025, parameterName=U_Offset}, B=ScalarParameter() {defaultValue=0.025, parameterName=V_Offset})
  n4 = Add(A=@n2, B=@n3)
  n5 = TextureSample(UVs=@n4) {texture=/Game/Effects/Fort_Effects/Textures/Plasma/T_Turbulence_Seamless}
  n6 = TextureSample(UVs=@n2) {texture=&T_Turbulence_Seamless}
  n7 = Add(A=@n2, B=Multiply(A=@n3) {constB=-1})
  n8 = TextureCoordinate() {coordinateIndex=1, uTiling=0.5}
  n9 = AppendVector(A=ScalarParameter() {defaultValue=0.025, parameterName=U_Offset_1}, B=ScalarParameter() {defaultValue=0.025, parameterName=V_Offset_1})
  n10 = Add(A=@n8, B=@n9)
  n11 = TextureSample(UVs=@n10) {texture=&T_Turbulence_Seamless}
  n12 = TextureSample(UVs=@n8) {texture=&T_Turbulence_Seamless}
  n13 = Add(A=@n8, B=Multiply(A=@n9) {constB=-1})
  n14 = LinearInterpolate(
   A=AppendVector(A=AppendVector(A=@n5.G, B=@n6.G), B=TextureSample(UVs=@n7) {texture=&T_Turbulence_Seamless}.G)
   B=AppendVector(A=AppendVector(A=@n11.G, B=@n12.G), B=TextureSample(UVs=@n13) {texture=&T_Turbulence_Seamless}.G)
  )
  n15 = Desaturation(None=@n14)
  n16 = Power(Base=@n15) {constExponent=0.5}
  n17 = LinearInterpolate(Alpha=Add(A=@n16, B=DynamicParameter() {paramNames [Erode,Param2,Param3,Param4]}.Erode)) {constA=-3}
  n18 = Multiply(A=@n1.A, B=Clamp(None=Multiply(A=@n16, B=@n17)))
  n19 = TextureCoordinate()
  n20 = ComponentMask(None=@n19) {r=true}
  n21 = ScalarParameter() {defaultValue=1, parameterName=Front Edge Fade}
  n22 = OneMinus(None=ComponentMask(None=@n19) {r=true})
  n23 = ScalarParameter() {defaultValue=2}
  n24 = Multiply(A=Sine(None=ComponentMask(None=@n19) {g=true}) {period=2}, B=Multiply(A=Power(Base=@n20, Exp=@n21), B=Power(Base=@n22, Exp=@n23)))
  n25 = AppendVector(
   A=TextureSample(UVs=@n4) {samplerType=SAMPLERTYPE_Normal, texture=/Game/Effects/Fort_Effects/Textures/Plasma/T_Turbulence_Seamless_N}.G
   B=TextureSample(UVs=@n2) {samplerType=SAMPLERTYPE_Normal, texture=&T_Turbulence_Seamless_N}.G
  )
  n26 = TextureSample(UVs=@n7) {samplerType=SAMPLERTYPE_Normal, texture=&T_Turbulence_Seamless_N}
  n27 = AppendVector(
   A=TextureSample(UVs=@n10) {samplerType=SAMPLERTYPE_Normal, texture=&T_Turbulence_Seamless_N}.G
   B=TextureSample(UVs=@n8) {samplerType=SAMPLERTYPE_Normal, texture=&T_Turbulence_Seamless_N}.G
  )
  n28 = TextureSample(UVs=@n13) {samplerType=SAMPLERTYPE_Normal, texture=&T_Turbulence_Seamless_N}
  n29 = Transform(None=LinearInterpolate(A=AppendVector(A=@n25, B=@n26.G), B=AppendVector(A=@n27, B=@n28.G)))
  n30 = LinearInterpolate(Alpha=Multiply(A=Multiply(A=@n18, B=@n24) {constA=1}, B=Fresnel(Normal=@n29))) {constA=1, constB=1.1}
  n31 = Constant() {r=1}
  FeatureLevelSwitch(Default=@n30, ES2=@n31, ES3_1=@n31, SM4_REMOVED=@n30)
  n32 = MaterialFunctionCall!(Use Alternate UVs, Alternate UVs) {materialFunction=/Game/Athena/Effects/Materials/Geode/Functions/MF_Geode_Crack_Colors}
  n33 = Saturate(None=@n32.Crack Color)
  n34 = TextureCoordinate()
  n35 = Power(Base=OneMinus(None=ComponentMask(None=@n34) {r=true})) {constExponent=5}
  n36 = TextureCoordinate() {coordinateIndex=1}
  n37 = TextureSample(UVs=Panner(Coordinate=@n36) {bFractionalPart=true, speedX=0.0312, speedY=0.0156}) {texture=/Game/Effects/Fort_Effects/Textures/Tile/T_Tiling_Dust}
  n38 = Multiply(A=@n37.R, B=LinearInterpolate(A=@n32.Cloud Color, B=@n32.Cloud Color, Alpha=@n15))
  n39 = Multiply(A=Add(A=LinearInterpolate(A=@n32.Cloud Color, B=@n33, Alpha=@n35), B=Add(A=@n38, B=@n14)), B=@n1.RGB)
  n40 = ScalarParameter() {defaultValue=5, parameterName=HDR_Intensity}
  n41 = MaterialFunctionCall(
   LDR Input=Clamp(None=@n39)
   HDR Tint=@n39
   HDR Intensity Multiplier=LinearInterpolate(A=@n40, B=Multiply(A=@n40) {constB=10}, Alpha=@n37.R) {constB=10}
  ) {materialFunction=&DeriveHDRfromLDR}
  FeatureLevelSwitch(Default=@n41.RemappedRange, ES2=@n39, ES3_1=@n39, SM4_REMOVED=@n41.RemappedRange)
  Panner(Coordinate=@n10) {speedX=-2, speedY=1}
  Panner(Coordinate=@n8) {speedX=-2, speedY=1}
  Panner(Coordinate=@n13) {speedX=-2, speedY=1}
  Panner(Coordinate=@n4) {speedX=-1, speedY=-0.5}
  Panner(Coordinate=@n2) {speedX=-1, speedY=-0.5}
  Panner(Coordinate=@n7) {speedX=-1, speedY=-0.5}
  Rotator(Coordinate=@n36) {speed=0.0312}
 ]
}
