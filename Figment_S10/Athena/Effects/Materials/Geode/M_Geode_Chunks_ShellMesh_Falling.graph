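# Masked subsurface material for mesh particles, apparently a falling shell effect for Geode chunks.
# Two downward-panning samples of an explosion normal map's green channel, tiled by UVScale, pass through Erode_Alpha with dynamic Erode, with inputs swapped on the second.
# Their product is multiplied by particle alpha and a UV-based SphereMask edge.
# Time-offset, actor-relative world position drives vector noise, scaled by dynamic curl parameters.
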Material /Figment_S10/Athena/Effects/Materials/Geode/M_Geode_Chunks_ShellMesh_Falling
graph {
 properties {
  bUsedWithMeshParticles=true
  blendMode=BLEND_Masked
  pixelDepthOffsetMode=PDOM_Legacy
  shadingModel=MSM_Subsurface
  usedShadingModels=Subsurface
 }
 nodes [
  Constant2Vector()
  n1 = Subtract(A=Subtract(A=WorldPosition().XYZ, B=ActorPositionWS()), B=Multiply(B=Time()))
  n2 = DynamicParameter() {paramNames [UVScale, Curl_Scale, Curl_Intensity, Erode]}
  Multiply(
   A=DynamicParameter() {paramNames [UVScale, Curl_Scale, Curl_Intensity, Erode]}.Curl_Intensity
   B=VectorNoise(World Position=Divide(A=@n1, B=@n2.Curl_Scale)) {noiseFunction=VNF_VectorALU}
  )
  n3 = DynamicParameter() {paramNames [UVScale, Curl_Scale, Curl_Intensity, Erode]}
  n4 = Multiply(A=DynamicParameter() {paramNames [UVScale, Curl_Scale, Curl_Intensity, Erode]}.UVScale, B=TextureCoordinate() {vTiling=2})
  n5 = MaterialFunctionCall!(
   Dynamic Param=@n3.Erode
   Alpha=TextureSample(UVs=Panner(Coordinate=@n4) {speedY=-0.15}) {samplerType=SAMPLERTYPE_Normal, texture=/Game/Effects/Fort_Effects/Gameplay/SmokeDustExplosions/Textures/T_Explosion_NM_01}.G
  ) {materialFunction=/Game/Effects/Fort_Effects/Materials/_Functions/Erode_Alpha}
  n6 = Multiply(A=TextureCoordinate() {uTiling=0.5}, B=DynamicParameter() {paramNames [UVScale, Curl_Scale, Curl_Intensity, Erode]}.UVScale)
  n7 = MaterialFunctionCall(
   Dynamic Param=TextureSample(UVs=Panner(Coordinate=@n6) {speedY=-0.12}) {samplerType=SAMPLERTYPE_Normal, texture=&T_Explosion_NM_01}.G
   Alpha=@n3.Erode
  ) {materialFunction=&Erode_Alpha}
  n8 = VectorParameter() {parameterName=MaskEdge, defaultValue=<rgba 0 0.9 0 0>}
  Multiply(
   A=Multiply(A=ParticleColor().A, B=Multiply(A=@n5.Result, B=@n7.Result))
   B=SphereMask(
    A=TextureCoordinate() {uTiling=0}
    B=ComponentMask(None=@n8.RGB) {g=true, r=true}
    Radius=ScalarParameter() {defaultValue=0.4, parameterName=MaskEdgeRadius}
    Hardness=ScalarParameter() {defaultValue=1, parameterName=MaskEdgeHardness}
   )
  )
  ParticleColor()
  Power()
  ScalarParameter() {parameterName=Metallic}
  ScalarParameter() {defaultValue=1, parameterName=Specular}
  ScalarParameter() {parameterName=Roughness}
  ScalarParameter() {defaultValue=15, parameterName=FadeDistance}
  SphereMask()
  VectorParameter() {parameterName=SubsurfaceColor, defaultValue=<rgba 0 0 0 1>}
 ]
}
