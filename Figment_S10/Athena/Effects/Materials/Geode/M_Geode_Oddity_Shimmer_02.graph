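# Translucent unlit particle sprite material for a Geode Oddity smoky shimmer, with color, opacity and refraction branches.
# Panning chromatic noise distorts a sphere-masked smoke flipbook, tinted by crack-color clouds and a scene-normal Fresnel.
# Depth fades confine opacity near scene geometry, while vertex color weights a sine-vibration flicker.
# A Sprite function drives camera-distance-scaled offsets; dithered refraction is switchable, and low feature levels zero out opacity and Fresnel.
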Material /Figment_S10/Athena/Effects/Materials/Geode/M_Geode_Oddity_Shimmer_02
graph {
 properties {
  bUsedWithParticleSprites=true
  blendMode=BLEND_Translucent
  pixelDepthOffsetMode=PDOM_Legacy
  refractionMethod=RM_IndexOfRefraction
  shadingModel=MSM_Unlit
  usedShadingModels=Unlit
 }
 nodes [
  n1 = SceneTexture() {sceneTextureId=PPI_WorldNormal}
  n2 = ScalarParameter() {defaultValue=500, parameterName=FadeInDistance}
  n3 = Multiply(
   A=Fresnel(Normal=ComponentMask(None=@n1.Color) {b=true, g=true, r=true}) {baseReflectFraction=0, exponent=1.5}
   B=OneMinus(None=DepthFade(FadeDistance=@n2) {fadeDistanceDefault=500})
  )
  n4 = Constant()
  n5 = ShadingPathSwitch(Default=@n3, Deferred=@n3, Forward=@n4, Mobile=@n3)
  n6 = FeatureLevelSwitch(Default=@n5, ES2=@n4, ES3_1=@n4, SM4_REMOVED=@n5)
  n7 = VertexColor()
  n8 = Normalize(VectorInput=Ceil(None=Add(A=@n7.R) {constB=4}))
  n9 = Constant() {r=2}
  n10 = TextureSample(UVs=Panner(Coordinate=TextureCoordinate() {coordinateIndex=1}) {speedX=-0.045, speedY=-0.09}) {texture=/Game/Effects/Fort_Effects/Textures/Tile/T_Inky_Smoke_Tile_Chromatic}
  n11 = ComponentMask(None=@n10.RGB) {g=true, r=true}
  n12 = Multiply(A=@n11, B=Constant2Vector() {g=0.05, r=0.02})
  n13 = TextureSample(UVs=Add(
    A=MaterialFunctionCall("Animation  Phase (0-1)"=@n8, Number of Rows=@n9, Number of Columns=@n9) {materialFunction=&FlipBook}.UVs
    B=@n12
   )) {texture=/Game/Effects/Fort_Effects/Textures/T_InkySmoke_Flipbook}
  Add(A=@n6, B=@n13.G)
  Constant()
  n14 = ScalarParameter() {defaultValue=0.11, parameterName=Opacity}
  n15 = Multiply(A=Time()) {constB=0.02}
  n16 = Constant() {r=5.5}
  n17 = Divide(
   A=Add(A=MaterialFunctionCall!(Time=@n15) {materialFunction=/Game/Effects/Fort_Effects/Materials/Geode/_Functions/MF_SineVibration}.Result, B=@n16) {constB=5}
   B=@n16
  )
  n18 = Add(A=Add(A=Multiply(A=@n7.R) {constB=0.8}, B=Multiply(A=@n7.G) {constB=0.5}), B=Multiply(A=@n7.B) {constB=0})
  n19 = TextureCoordinate()
  n20 = Constant() {r=0.5}
  n21 = Multiply(A=@n13.G, B=SphereMask(A=Add(A=@n12, B=@n19), B=@n20, Radius=@n20) {hardnessPercent=0})
  n22 = DepthFade(Opacity=@n21) {fadeDistanceDefault=50}
  n23 = Multiply(
   A=Multiply(A=ParticleColor().A, B=LinearInterpolate(B=@n17, Alpha=@n18) {constA=1})
   B=FeatureLevelSwitch(Default=@n22, ES2=@n21, ES3_1=@n21, SM4_REMOVED=@n22)
  )
  n24 = DepthFade(FadeDistance=ScalarParameter() {defaultValue=550, parameterName=VisibleDepthDistance}) {fadeDistanceDefault=80}
  n25 = DepthFade(
   Opacity=Multiply(A=Multiply(A=@n23, B=@n14), B=OneMinus(None=@n24))
   FadeDistance=ScalarParameter() {defaultValue=200, parameterName=DepthFade}
  ) {fadeDistanceDefault=15}
  n26 = Constant()
  FeatureLevelSwitch(Default=@n25, ES2=@n26, ES3_1=@n26, SM4_REMOVED=@n25, SM5=@n25, SM6=@n25)
  n27 = Multiply(A=@n6)
  n28 = MaterialFunctionCall!(Use Alternate UVs, Alternate UVs) {materialFunction=/Game/Athena/Effects/Materials/Geode/Functions/MF_Geode_Crack_Colors}
  n29 = TextureCoordinate() {coordinateIndex=1}
  n30 = Multiply(A=@n11, B=Constant2Vector() {g=-0.1, r=-0.02})
  n31 = TextureSample(UVs=Panner(Coordinate=Add(A=@n29, B=@n30)) {speedX=0.05, speedY=0.12}) {texture=&T_Inky_Smoke_Tile_Chromatic}
  n32 = LinearInterpolate(A=Multiply(A=@n21, B=@n21), B=@n31.G) {constAlpha=0.7}
  Multiply(
   A=Multiply(A=Add(B=@n27) {constA=1}, B=VectorParameter() {parameterName=color, defaultValue=<rgba 0.6371 0.5527 0.8698 1>}.RGB)
   B=Add(A=@n10.RGB, B=LinearInterpolate(A=@n28.Cloud Color, B=@n28.Cloud Color, Alpha=@n32))
  )
  n33 = LinearInterpolate(
   A=ScalarParameter() {defaultValue=1400, parameterName=SpriteScaleNear}
   B=ScalarParameter() {defaultValue=1400, parameterName=SpriteScaleFar}
   Alpha=MaterialFunctionCall(Distance=ScalarParameter() {defaultValue=580, parameterName=SoftenDistance}) {materialFunction=&Near_Camera_Fade}.Result
  )
  n34 = Multiply(A=@n33) {constB=0.5}
  n35 = Multiply(A=@n33) {constB=0.16}
  Reroute(None=MaterialFunctionCall!(Center Location=WorldPosition().XYZ, 0-1 UVs, XY scale=AppendVector(A=@n34, B=@n35), OverridePivot, "Normals (Optional)") {materialFunction=&Sprite}.World position offset)
  n36 = Add(A=ScalarParameter() {parameterName=RefractionAdd})
  n37 = Add(A=@n23, B=Multiply(A=@n23, B=Multiply(A=@n31.R) {constB=2}))
  StaticSwitchParameter(
   True=Constant() {r=1}
   False=MaterialFunctionCall!(RefractiveIndex=LinearInterpolate(B=@n36, Alpha=@n37) {constA=1}) {materialFunction=/Game/Effects/Fort_Effects/Materials/_Functions/MF_Dithered_Refraction}.Result
  ) {parameterName=Remove_Refraction}
  TextureSample() {samplerType=SAMPLERTYPE_Normal, texture=/Game/Effects/Fort_Effects/Textures/Mask/T_Sphere_N}
 ]
}
