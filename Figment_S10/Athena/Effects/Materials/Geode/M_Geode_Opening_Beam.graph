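# Additive, two-sided unlit mesh-particle material for a light beam, apparently for the Geode opening effect.
# World position offset combines SplineThicken, camera offset and outward scaling, with width and scale growing by camera distance to the object.
# Panning smoke noise tints an HDR color, with a separate mobile color.
# Opacity multiplies UV and sphere masks, near-camera fade and non-mobile depth fade by timestamp-driven fade in/out, which a debug switch can override.
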Material /Figment_S10/Athena/Effects/Materials/Geode/M_Geode_Opening_Beam
graph {
 properties {
  bUsedWithMeshParticles=true
  blendMode=BLEND_Additive
  pixelDepthOffsetMode=PDOM_Legacy
  shadingModel=MSM_Unlit
  twoSided=true
  usedShadingModels=Unlit
 }
 nodes [
  n1 = ObjectPositionWS()
  n2 = Distance(A=@n1, B=CameraPositionWS())
  n3 = ScalarParameter() {defaultValue=500, parameterName=DistanceOffset}
  n4 = Clamp(None=Divide(A=Subtract(A=@n2, B=@n3), B=ScalarParameter() {defaultValue=30000, parameterName=ScaleUpDistance}))
  n5 = Multiply(
   A=ScalarParameter() {defaultValue=150, parameterName=SplineWidth}
   B=LinearInterpolate(B=ScalarParameter() {defaultValue=10, parameterName=MaxWidthMult}, Alpha=@n4) {constA=1, constB=3}
  )
  n6 = Multiply(A=Multiply(A=ScalarParameter() {defaultValue=75, parameterName=MaxScaleIncrease}) {constB=-1}, B=Subtract(A=@n1, B=WorldPosition().XYZ)) {constB=-1}
  Add(
   A=Add(
    A=MaterialFunctionCall(WidthBase=@n5, WidthTip=@n5) {materialFunction=&SplineThicken}.WorldPositionOffset
    B=LinearInterpolate(B=@n6, Alpha=@n4) {constB=0}
   )
   B=MaterialFunctionCall(Offset Amount=ScalarParameter() {defaultValue=400, parameterName=Camera Offset Amount}, World Position=WorldPosition().XYZ) {materialFunction=/Engine/Functions/Engine_MaterialFunctions02/WorldPositionOffset/CameraOffset}.Camera Offset
  )
  n7 = VectorParameter() {parameterName=Color, defaultValue=<rgba 1.5 1.175 0.75 1>}
  n8 = TextureCoordinate() {vTiling=5}
  n9 = TextureCoordinate() {vTiling=5}
  n10 = Add(
   A=TextureSample(UVs=Panner(Coordinate=@n8) {speedX=-0.01, speedY=-0.04}) {texture=/Game/Effects/Fort_Effects/Textures/Tile/T_Inky_Smoke_Tile_Chromatic}.R
   B=TextureSample(UVs=Panner(Coordinate=@n9) {speedX=-0.015, speedY=-0.1}) {texture=&T_Inky_Smoke_Tile_Chromatic}.G
  )
  n11 = MaterialFunctionCall(
   LDR Input=Clamp(None=@n7.RGB)
   HDR Tint=Add(A=@n10, B=@n7.RGB)
   HDR Intensity Multiplier=ScalarParameter() {defaultValue=1, parameterName=HDRIntensity}
  ) {materialFunction=&DeriveHDRfromLDR}
  n12 = Add(A=@n10, B=VectorParameter() {parameterName=Color_Mobile, defaultValue=<rgba 15 7.4362 0 1>}.RGB)
  FeatureLevelSwitch(Default=@n11.RemappedRange, ES2=@n12, ES3_1=@n12, SM4_REMOVED=@n11.RemappedRange)
  n13 = ScalarParameter() {parameterName=FadeInTime}
  n14 = ScalarParameter() {defaultValue=2, parameterName=Fade In Time Length}
  n15 = ScalarParameter() {parameterName=FadeOutTime}
  n16 = ScalarParameter() {defaultValue=2, parameterName=Fade Out Time Length}
  n17 = StaticBool()
  n18 = ComponentMask(None=TextureCoordinate() {coordinateIndex=1}) {g=true}
  n19 = Multiply(A=@n18, B=OneMinus(None=@n18))
  n20 = ComponentMask(None=TextureCoordinate()) {g=true}
  n21 = SphereMask(
   A=Power(Base=TextureCoordinate() {uTiling=0}) {constExponent=0.6}
   B=Constant2Vector() {g=0.55}
   Radius=ScalarParameter() {defaultValue=0.45, parameterName=Radius_1}
   Hardness=ScalarParameter() {defaultValue=-5, parameterName=Hardness_1}
  )
  n22 = SphereMask(
   A=TextureCoordinate() {coordinateIndex=1, uTiling=0}
   B=Constant2Vector() {g=0.5}
   Radius=ScalarParameter() {defaultValue=0.5, parameterName=Radius}
   Hardness=ScalarParameter() {parameterName=Hardness}
  )
  n23 = MaterialFunctionCall(Distance=ScalarParameter() {defaultValue=25000, parameterName="Distance (S)"}) {materialFunction=&Near_Camera_Fade}
  n24 = ScalarParameter() {defaultValue=1.5, parameterName=CamFadeExp}
  n25 = Multiply(A=Multiply(A=Multiply(A=@n19, B=@n20), B=Multiply(A=@n21, B=@n22)), B=Clamp(None=Power(Base=@n23.Result, Exp=@n24)))
  n26 = DepthFade(Opacity=@n25, FadeDistance=ScalarParameter() {defaultValue=300, parameterName=DepthFadeDistance}) {fadeDistanceDefault=200}
  Multiply(
   A=StaticSwitchParameter(
    True=ScalarParameter() {defaultValue=1, parameterName=DEBUG_Alpha}
    False=MaterialFunctionCall(TimestampA=@n13, TransitionLength_In=@n14, TimestampB=@n15, TransitionLength_Out=@n16, Transition In Only?=@n17) {materialFunction=&MF_Timestamp_InOut_01}.Linear
   ) {parameterName=DEBUG}
   B=FeatureLevelSwitch(Default=@n26, ES2=@n25, ES3_1=@n25, SM4_REMOVED=@n26)
  )
 ]
}
