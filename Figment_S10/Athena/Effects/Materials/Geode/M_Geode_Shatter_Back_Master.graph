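# Translucent, unlit, two-sided master material for mesh particles, apparently the back layer of a geode shatter effect.
# Color adds a scaled geode cubemap to particle-tinted glass cracks, using shared crack and cloud colors.
# A shatter normal on UV1 blends additively with crack normals; low quality drops the crack color and normal layers.
# Faint opacity combines particle alpha, vertex color, a V gradient and a UV1 sphere mask driven by a dynamic parameter.
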Material /Figment_S10/Athena/Effects/Materials/Geode/M_Geode_Shatter_Back_Master
graph {
 properties {
  bUsedWithMeshParticles=true
  blendMode=BLEND_Translucent
  pixelDepthOffsetMode=PDOM_Legacy
  shadingModel=MSM_Unlit
  twoSided=true
  usedShadingModels=Unlit
 }
 nodes [
  n1 = TextureCoordinate() {uTiling=2}
  n2 = MaterialFunctionCall!(Use Alternate UVs, Alternate UVs) {materialFunction=/Game/Athena/Effects/Materials/Geode/Functions/MF_Geode_Crack_Colors}
  n3 = TextureSample(UVs=@n1) {samplerType=SAMPLERTYPE_Normal, texture=/Game/Effects/Fort_Effects/Textures/Tile/T_Glass_Cracks_N}
  n4 = Clamp(None=DotProduct(A=@n3.RGB, B=Constant3Vector() {constant=<rgba 0 1 0 0>}))
  n5 = ComponentMask(None=TextureCoordinate()) {g=true}
  n6 = Multiply(A=Multiply(A=@n2.Crack Color) {constB=5}, B=Power(Base=@n5) {constExponent=3.5})
  n7 = Multiply(
   A=Multiply(A=TextureSample(UVs=@n1) {texture=/Game/Effects/Fort_Effects/Textures/Tile/T_Glass_Cracks_M}.G) {constB=2}
   B=Add(A=Multiply(A=@n2.Cloud Color) {constB=0.4}, B=Multiply(A=@n4, B=@n6) {constB=2})
  )
  n8 = ParticleColor()
  Add(
   A=Multiply(
    A=MaterialFunctionCall!(UV add) {materialFunction=/Game/Athena/Effects/Materials/Geode/Functions/MF_Geode_Cubemap}.Result
    B=ScalarParameter() {defaultValue=0.6, parameterName=LerpCube}
   )
   B=QualitySwitch(Default=Multiply(A=@n7, B=@n8.RGB), Low=Constant())
  )
  n9 = VectorParameter() {parameterName=CrackNormalMult, defaultValue=<rgba 0.1 0.1 0 1>}
  n10 = TextureSampleParameter2D(UVs=TextureCoordinate() {coordinateIndex=1}) {
   parameterName=ShatterNormalTexture
   samplerType=SAMPLERTYPE_Normal
   texture=/Figment_S10/Athena/Effects/Textures/Geode/T_Geode_Test_Spline_01_N
  }
  Add(
   A=QualitySwitch(Default=Multiply(A=@n3.RGB, B=@n9.RGB), Low=Constant())
   B=Multiply(A=@n10.RGB, B=VectorParameter() {parameterName=ScaleNormal, defaultValue=<rgba 1 1 4 0>}.RGB)
  )
  n11 = ScalarParameter() {defaultValue=1, parameterName=Opacity}
  n12 = VertexColor()
  n13 = OneMinus(None=@n12.B)
  n14 = VectorParameter() {parameterName=OpacityMaskCenter, defaultValue=<rgba 0.4381 0.419 0 0>}
  n15 = LinearInterpolate(
   B=DynamicParameter() {paramNames [Opacity,Param2,Param3,Param4]}.Opacity
   Alpha=Power(Base=ScalarParameter() {defaultValue=1, parameterName=CrackSize})
  )
  n16 = Multiply(
   A=Multiply(A=@n5, B=Multiply(A=@n12.R, B=@n13))
   B=SphereMask(
    A=TextureCoordinate() {coordinateIndex=1}
    B=ComponentMask(None=@n14.RGB) {g=true, r=true}
    Radius=LinearInterpolate(Alpha=@n15) {constB=1.25}
    Hardness=LinearInterpolate(Alpha=@n15) {constB=0.5}
   )
  )
  Multiply(A=Multiply(A=Power(Base=@n11) {constExponent=0.6}) {constB=0.15}, B=Multiply(A=@n8.A, B=Clamp(None=@n16)))
 ]
}
